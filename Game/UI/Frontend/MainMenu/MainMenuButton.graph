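# Main menu button widget, based on FortMainMenuButton, with icon, uppercase label, optional right-side external-link icon and input action prompt.
# PreConstruct shows the right icon from ShowExternalLinkIcon and caps label width on mobile.
# When that flag is set, hover or selection darkens the icon and unhover restores light blue.
# Custom events swap its material or color, while input icon updates set dark, outline-free material parameters.
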WidgetBlueprint /Game/UI/Frontend/MainMenu/MainMenuButton
graphs [EventGraph {nodes [
   "On Input Icon Updated (InputActionWidget)"() {is=title K2Node_ComponentBoundEvent}
   > n1 = GetIconDynamicMaterial!(execute, self:Common Action Widget Object Reference=Get InputActionWidget() -> (InputActionWidget:Common Action Widget Object Reference)) -> (then, ReturnValue:Material Instance Dynamic Object Reference)
   > SetScalarParameterValue(self=@n1, ParameterName=Use Dark, Value=1)
   > SetScalarParameterValue(self=@n1, ParameterName=Outline Opacity, Value=0)
   n2 = UpdateRightSideIcon!() -> (OutputDelegate, then, SoftMaterial:Material Interface Soft Object Reference) {is=title K2Node_CustomEvent}
   n3 = Get Image_Icon_RightSide() -> (Image_Icon_RightSide:Image Object Reference)
   > SetBrushFromSoftMaterial!(execute, self:Image Object Reference=@n3, SoftMaterial:Material Interface Soft Object Reference=@n2.SoftMaterial) -> (then)
   > SetVisibility(self=@n3, InVisibility=SelfHitTestInvisible)
   n4 = UpdateRightSideIconColor!() -> (OutputDelegate, then, Color:Linear Color Structure) {is=title K2Node_CustomEvent}
   > SetColorAndOpacity(self:Image Object Reference=Get Image_Icon_RightSide(), InColorAndOpacity:Linear Color Structure=@n4.Color)
   Event PreConstruct()
   > SetVisibility(
    self=Get Image_Icon_RightSide()
    InVisibility=Select(Option 0=Collapsed, Option 1=SelfHitTestInvisible, Index=Get ShowExternalLinkIcon() -> (ShowExternalLinkIcon:bool))
   )
   > Branch(Condition=IsMobileGame())
   > SetMaxDesiredWidth!(execute, self:Size Box Object Reference=Get SizeBox_Mobile() -> (SizeBox_Mobile:Size Box Object Reference), InMaxDesiredWidth:float=450) -> (then)
   n5 = Event On Hovered()
   Branch(execute=[@n5.then, @n6.then], Condition=Get ShowExternalLinkIcon())
   > UpdateRightSideIconColor!(execute, self, Color:Linear Color Structure=(R=0,G=0.004,B=0.0723,A=1)) -> (then) {is=title K2Node_CallFunction}
   n6 = Event On Selected()
   Event On Unhovered()
   > Branch(Condition=Get ShowExternalLinkIcon())
   > UpdateRightSideIconColor(Color=(R=0.3185,G=0.7231,B=1,A=1))
   UpdateRightSideIconColor(Color=(R=0.0513,G=0.0513,B=0.0513,A=1))
  ]}]
variables [ShowExternalLinkIcon {type=boolean}]
widgets {
 info {inheritedWidgetCount=4, parentClass=/Script/FortniteUI.FortMainMenuButton, rootWidgetClass=&Border, widgetCount=7}
 widgets [
  ContentBorder {
   widgetClassPath=&Border
   bInherited=true
   widget=ContentBorder
   properties {
    horizontalAlignment=HAlign_Left
    background {
     bIsDynamicallyLoaded=false
     drawAs=NoDrawType
     imageType=NoImage
     mirroring=NoMirror
     resourceName=None
     resourceObject=null
     tiling=NoTile
     imageSize=<32 32>
     margin=<ltrb 0 0 0 0>
     outlineSettings {
      bUseBrushTransparency=false
      roundingType=HalfHeightRadius
      width=0
      color {colorUseRule=UseColor_Specified, specifiedColor=<rgba 0 0 0 0>}
      cornerRadii=<0 0 0 0>
     }
     tintColor {colorUseRule=UseColor_Specified, specifiedColor=<rgba 1 1 1 1>}
     uVRegion {bIsValid=false, max=<0 0>, min=<0 0>}
    }
    padding=<ltrb 0 0 0 0>
   }
  }
  HorizontalBox {
   widgetClassPath=&HorizontalBox
   parent=ContentBorder
   widget=HorizontalBox
   properties {flowDirectionPreference=Culture}
   slot {class=BorderSlot, properties {horizontalAlignment=HAlign_Left, padding=<ltrb 0 0 0 0>}}
  }
  Image_Icon {
   widgetClassPath=&Image
   bInherited=true
   parent=HorizontalBox
   widget=Image_Icon
   properties {brush {
     bIsDynamicallyLoaded=false
     drawAs=NoDrawType
     imageType=NoImage
     mirroring=NoMirror
     resourceName=None
     resourceObject=null
     tiling=NoTile
     imageSize=<0 0>
     margin=<ltrb 0 0 0 0>
     outlineSettings {
      bUseBrushTransparency=false
      roundingType=HalfHeightRadius
      width=0
      color {colorUseRule=UseColor_Specified, specifiedColor=<rgba 0 0 0 0>}
      cornerRadii=<0 0 0 0>
     }
     tintColor {colorUseRule=UseColor_Specified, specifiedColor=<rgba 1 1 1 1>}
     uVRegion {bIsValid=false, max=<0 0>, min=<0 0>}
    }}
   slot {class=HorizontalBoxSlot, properties {horizontalAlignment=HAlign_Center, verticalAlignment=VAlign_Center, padding=<ltrb 4 4 4 4>}}
  }
  SizeBox_Mobile {
   widgetClassPath=&SizeBox
   bIsVariable=true
   parent=HorizontalBox
   widget=SizeBox_Mobile
   slot {class=HorizontalBoxSlot, properties {verticalAlignment=VAlign_Center}}
  }
  Text_Content {
   widgetClassPath=&CommonTextBlock
   bInherited=true
   bIsVariable=true
   parent=SizeBox_Mobile
   widget=Text_Content
   properties {
    mobileFontSizeMultiplier=1.2
    scrollStyle=/FortUILibrary/Styles/Typography/Foundation/TextScrollStyle-NoFade.TextScrollStyle-NoFade_C
    style=/FortUILibrary/Foundation/Buttons/ButtonStyles/Athena/TextStyle_BurbSmBk_15_LightBlue.TextStyle_BurbSmBk_15_LightBlue_C
    text=Menu Settings
    textTransformPolicy=ToUpper
    colorAndOpacity {colorUseRule=UseColor_Specified, specifiedColor=<rgba 0.3185 0.7231 1 1>}
    font {
     bForceMonospaced=false
     bMaterialIsStencil=false
     fontMaterial=null
     fontObject=&BurbankSmall
     letterSpacing=0
     monospacedWidth=1
     size=15
     skewAmount=0
     typefaceFontName=Black
     outlineSettings {
      bApplyOutlineToDropShadows=false
      bMiteredCorners=false
      bSeparateFillAlpha=false
      outlineMaterial=null
      outlineSize=0
      outlineColor=<rgba 0 0 0 1>
     }
    }
    shadowOffset=<0 0>
   }
   slot {class=SizeBoxSlot}
  }
  Image_Icon_RightSide {
   widgetClassPath=&Image
   bIsVariable=true
   parent=HorizontalBox
   widget=Image_Icon_RightSide
   properties {
    visibility=Collapsed
    brush {
     bIsDynamicallyLoaded=false
     drawAs=Image
     imageType=NoImage
     mirroring=NoMirror
     resourceName=None
     resourceObject=/FortUILibrary/Foundation/Buttons/Materials/M_UI_Icon_ExternalLink
     tiling=NoTile
     imageSize=<32 32>
     margin=<ltrb 0 0 0 0>
     outlineSettings {
      bUseBrushTransparency=false
      roundingType=HalfHeightRadius
      width=0
      color {colorUseRule=UseColor_Specified, specifiedColor=<rgba 0 0 0 0>}
      cornerRadii=<0 0 0 0>
     }
     tintColor {colorUseRule=UseColor_Specified, specifiedColor=<rgba 1 1 1 1>}
     uVRegion {bIsValid=false, max=<0 0>, min=<0 0>}
    }
    colorAndOpacity=<rgba 0.3185 0.7231 1 1>
   }
   slot {class=HorizontalBoxSlot, properties {horizontalAlignment=HAlign_Left, verticalAlignment=VAlign_Center, padding=<ltrb 4 4 4 4>}}
  }
  InputActionWidget {
   widgetClassPath=&CommonActionWidget
   bInherited=true
   parent=HorizontalBox
   widget=InputActionWidget
   properties {
    progressMaterialParam=Percentage
    progressMaterialBrush {
     bIsDynamicallyLoaded=false
     drawAs=Image
     imageType=NoImage
     mirroring=NoMirror
     resourceName=None
     resourceObject=/Game/UI/Materials/HollowCircularProgressBar
     tiling=NoTile
     imageSize=<36 36>
     margin=<ltrb 0 0 0 0>
     outlineSettings {
      bUseBrushTransparency=false
      roundingType=HalfHeightRadius
      width=0
      color {colorUseRule=UseColor_Specified, specifiedColor=<rgba 0 0 0 0>}
      cornerRadii=<0 0 0 0>
     }
     tintColor {colorUseRule=UseColor_Specified, specifiedColor=<rgba 1 1 0 1>}
     uVRegion {bIsValid=false, max=<0 0>, min=<0 0>}
    }
   }
   slot {
    class=HorizontalBoxSlot
    properties {horizontalAlignment=HAlign_Right, verticalAlignment=VAlign_Center, padding=<ltrb 4 0 4 0>, size {sizeRule=Fill, value=1}}
   }
  }
 ]
}
